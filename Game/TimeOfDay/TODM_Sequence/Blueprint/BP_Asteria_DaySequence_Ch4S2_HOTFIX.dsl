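; Day sequence Blueprint for Asteria (Chapter 4 Season 2 hotfix) that feeds time-of-day values to materials and gates volumetric fog.
; SetMPCValues copies real time, sun color and axes, and height fog colors into the global material parameter collection.
; Custom events drive storm cloud time and phase in the safe zone cloud collection.
; Setting the cloud material activates or unloads the VolumeFog data layer by cloud quality, excluding Switch and mobile.
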(event Collision|EventActorBeginOverlap
  (params (OtherActor :type Actor%20Object%20Reference)))

(event EventTick
  (params (DeltaSeconds :type Float%20%28single-precision%29))
  (|CallParentFunction DeltaSeconds) ; This node is disabled and will not be called.\nDrag off pins to build functionality.
)

(event EventSetVolumetricCloudMaterial
  ;[1 1 1] Validate Volume Fog
  (|CallParentFunction)
  (CallFunction|ValidateVolumeFog))

(event EventOnTimeOfDayManagerVisibilitySet
  (params (bVisible :type Boolean))
  (|CallParentFunction bVisible))

(event Custom|SetCloudsCustomTime
  (params (customTime :type Float%20%28double-precision%29))
  (Rendering|Material|SetScalarParameterValue "/Game/Athena/SafeZone/MPC_SafeZoneClouds.MPC_SafeZoneClouds" "CustomCloudsTime" customTime :!class /Script/Engine.KismetMaterialLibrary))

(event Custom|SetStormPhase
  (params (StormPhase :type Float%20%28double-precision%29))
  (Rendering|Material|SetScalarParameterValue "/Game/Athena/SafeZone/MPC_SafeZoneClouds.MPC_SafeZoneClouds" "StormPhase" StormPhase :!class /Script/Engine.KismetMaterialLibrary))

(event Custom|SetStormPhaseCount
  (params (StormPhaseCount :type Integer :default 5))
  (Rendering|Material|SetScalarParameterValue "/Game/Athena/SafeZone/MPC_SafeZoneClouds.MPC_SafeZoneClouds" "StormPhaseCount" (Math|Conversions|ToFloat(Integer) StormPhaseCount) :!class /Script/Engine.KismetMaterialLibrary))

(event Custom|SetSkyLightCubeMapBlend_1
  (params (val :type Float%20%28double-precision%29)))

(event Sequence|EventUpdateScalabilitySequences
  (CallFunction|UpdateScalability))

(event Scalability|EventOnGlobalIlluminationQualityChanged
  (|CallParentFunction)
  (CallFunction|UpdateScalability))

(fn SetMPCValues
  (Rendering|Material|SetScalarParameterValue "/Game/Packages/Fortress_SharedMaterials/GlobalMaterialParameters/FortniteMaterialParameters.FortniteMaterialParameters" "RealGameTime" (Utilities|Time|GetRealTimeSeconds) :!class /Script/Engine.KismetMaterialLibrary)
  (Rendering|Material|SetVectorParameterValue "/Game/Packages/Fortress_SharedMaterials/GlobalMaterialParameters/FortniteMaterialParameters.FortniteMaterialParameters" "SunLightColor" (Math|Conversions|ToLinearColor(Color) (Variables|/Script/Engine.LightComponentBase|GetLightColor (Variables|DayNightCycle|GetSunComponent))) :!class /Script/Engine.KismetMaterialLibrary)
  (Rendering|Material|SetVectorParameterValue "/Game/Packages/Fortress_SharedMaterials/GlobalMaterialParameters/FortniteMaterialParameters.FortniteMaterialParameters" "FogColor" (Variables|/Script/Engine.ExponentialHeightFogComponent|GetFogInscatteringLuminance (Variables|DayNightCycle|GetExponentialHeightFogComponent)) :!class /Script/Engine.KismetMaterialLibrary)
  (Rendering|Material|SetVectorParameterValue "/Game/Packages/Fortress_SharedMaterials/GlobalMaterialParameters/FortniteMaterialParameters.FortniteMaterialParameters" "FogDirectionalInscatteringColor" (Variables|/Script/Engine.ExponentialHeightFogComponent|GetDirectionalInscatteringLuminance (Variables|DayNightCycle|GetExponentialHeightFogComponent)) :!class /Script/Engine.KismetMaterialLibrary)
  (Rendering|Material|SetVectorParameterValue "/Game/Packages/Fortress_SharedMaterials/GlobalMaterialParameters/FortniteMaterialParameters.FortniteMaterialParameters" "SunlightYTransformVector" (Math|Conversions|ToLinearColor(Vector) (Transformation|GetRightVector (Variables|DayNightCycle|GetSunComponent))) :!class /Script/Engine.KismetMaterialLibrary)
  (Rendering|Material|SetVectorParameterValue "/Game/Packages/Fortress_SharedMaterials/GlobalMaterialParameters/FortniteMaterialParameters.FortniteMaterialParameters" "SunlightZTransformVector" (Math|Conversions|ToLinearColor(Vector) (Transformation|GetUpVector (Variables|DayNightCycle|GetSunComponent))) :!class /Script/Engine.KismetMaterialLibrary)
  (Rendering|Material|SetVectorParameterValue "/Game/Packages/Fortress_SharedMaterials/GlobalMaterialParameters/FortniteMaterialParameters.FortniteMaterialParameters" "SunlightDirectionalVector" (Math|Conversions|ToLinearColor(Vector) (* (Transformation|GetForwardVector (Variables|DayNightCycle|GetSunComponent)) -1.0)) :!class /Script/Engine.KismetMaterialLibrary))

(fn UpdateScalability
  (|CallParentFunction))

(fn ConstructionScript
  (|CallParentFunction))

(fn ValidateVolumeFog
  (Networking|CanExecuteCosmeticEvents
    (:True
      (bind _returnvalue (Development|GetConsoleVariableIntValue "Athena.VolumetricClouds.Quality"))
      (bind _returnvalue_2 (WorldPartition|GetDataLayerManager))
      (DataLayers|SetDataLayerInstanceRuntimeState _returnvalue_2 (DataLayers|GetDataLayerInstancefromAsset _returnvalue_2 "/Game/Athena/Asteria/DataLayers/Asteria_DL_VolumeFog.Asteria_DL_VolumeFog") (select (Math|Boolean|ANDBoolean (>= _returnvalue 2) (not (TouchPlatform|IsNintendoSwitch)) (not (TouchPlatform|IsMobileGame))) "Activated" "Unloaded") false))))
