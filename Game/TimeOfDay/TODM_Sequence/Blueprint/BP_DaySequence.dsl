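; Fortnite time-of-day day sequence actor that adapts sky, post-process and material globals to scalability settings.
; On begin play, visibility changes, quality events and data-driven CVar changes, it toggles volumetric clouds, convolution bloom and exposure volumes, and writes shadow, reflection and GI flags to a shared material parameter collection.
; A Lumen support check gates bloom and exposure choices.
; Cloud materials and bloom kernels load async, or blocking in editor.
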(event EventBeginPlay
  ;[0.1244 0.3385 0.105 0.464 48] Begin Play - Scalability Set Up for in game changes
  (Networking|CanExecuteCosmeticEvents
    (:True
      (bind _returnvalue (Settings|GetGameUserSettings))
      (EventDispatchers|BindEventtoOnGameUserSettingsUINeedsUpdate _returnvalue (EventDispatchers|CreateEvent :!event UpdateScalabilitySequences))
      (DataDrivenCVar|BindEventtoOnDataDrivenCVarDelegate (EngineSubsystems|GetDataDrivenCVarEngineSubsystem) (delegate Custom|OnDataDrivenCVarDelegate))
      (CallFunction|UpdateComponentVisibility)
      (Sequence|UpdateScalabilitySequences)
      ;[1 0 0.1306] Remove Later With Code Implementation
      (Development|ExecuteConsoleCommand (Utilities|String|Append "r.Shadow.Virtual.Cache.ForceInvalidateDirectional " (Utilities|String|ToString(Integer) (Variables|VSM|GetInvalidateDirectional)))))))

(event Custom|OnDataDrivenCVarDelegate
  (params (CVarName :type String))
  ;[0.1244 0.3385 0.105 0.464 48] Begin Play - Scalability Set Up for in game changes
  (Utilities|FlowControl|Sequence
    (:then_0
      (if (Utilities|String|Equal,CaseInsensitive(String) CVarName "Athena.VolumetricClouds.Quality")
        (CallFunction|EventSetVolCloudMaterial)))
    (:then_1
      (if (Utilities|String|Equal,CaseInsensitive(String) CVarName "Athena.PostProcessLevel")
        (CallFunction|EventSetCustomBloom)))
    (:then_2
      (bind _isexposurecvar (CallFunction|IsExposureCvar CVarName))
      (if _isexposurecvar
        (CallFunction|SetExposureOverrides)))))

(event EventOnTimeOfDayManagerVisibilitySet
  (params (bVisible :type Boolean))
  ;[1 1 1 1 48] Trigger Scalability Settings on Visibility Set
  (if bVisible
    (CallFunction|UpdateComponentVisibility)
    (Sequence|UpdateScalabilitySequences)))

(event Sequence|EventUpdateScalabilitySequences
  ;[1 1 1 1 48] Trigger Scalability Settings on Visibility Set
  (CallFunction|UpdateScalability))

(event EventEndPlay
  (params (EndPlayReason :type EEndPlayReason%20Enum :default "Destroyed"))
  ;[1 0 0.1057] Remove Later with code implementation
  (Development|ExecuteConsoleCommand "r.Shadow.Virtual.Cache.ForceInvalidateDirectional 0"))

(event Scalability|EventOnEffectsQualityChanged
  ;[0.1828 0.3073 0.27 1 48] Bind events for scalability
  (CallFunction|EventSetVolCloudMaterial))

(event Scalability|EventOnShadowQualityChanged
  ;[0.1828 0.3073 0.27 1 48] Bind events for scalability
  (CallFunction|SetShadowMPC))

(event Scalability|EventOnPostProcessQualityChanged
  ;[0.1828 0.3073 0.27 1 48] Bind events for scalability
  (CallFunction|EventSetCustomBloom))

(event Scalability|EventOnReflectionQualityChanged
  ;[0.1828 0.3073 0.27 1 48] Bind events for scalability
  (CallFunction|SetReflectionMPC))

(event Scalability|EventOnGlobalIlluminationQualityChanged
  ;[0.1828 0.3073 0.27 1 48] Bind events for scalability
  (CallFunction|SetLowEndExposure)
  (CallFunction|EventSetCustomBloom)
  (CallFunction|SetGlobalIlluminationMPC))

(event Custom|EventSetCustomBloom
  (bind _postprocess_bloomconvoution (Variables|Default|GetPostProcess_BloomConvoution))
  (bind _postprocess_bloomconvoution_2 (Variables|Default|GetPostProcess_BloomConvoution))
  ;[0.2656 0.2656 0.2656 1 48] Custom Bloom
  (bind _returnvalue (Development|GetConsoleVariableIntValue "Athena.PostProcessLevel"))
  (bind _islumensupported (CallFunction|IsLumenSupported))
  (if (and _islumensupported (>= _returnvalue 3))
    (if (Deprecated|UEFN|IsinEditor)
      (bind _returnvalue_2 (Utilities|LoadAssetBlocking (Variables|Bloom|GetConvolution%20Kernel)))
      (bind _astexture_2d (Utilities|Casting|CastToTexture2D _returnvalue_2))
      (bind _structout (Utilities|Struct|SetmembersinPostProcessSettings (Variables|/Script/Engine.PostProcessComponent|GetSettings (Variables|Default|GetPostProcess_BloomConvoution)) _astexture_2d))
      (Variables|/Script/Engine.PostProcessComponent|SetSettings _structout (Variables|Default|GetPostProcess_BloomConvoution))
      (else
        (bind _object (Utilities|AsyncLoadAsset (Variables|Bloom|GetConvolution%20Kernel))
          (:Completed
            (bind _astexture_2d_2 (Utilities|Casting|CastToTexture2D _object))
            (bind _structout_2 (Utilities|Struct|SetmembersinPostProcessSettings (Variables|/Script/Engine.PostProcessComponent|GetSettings (Variables|Default|GetPostProcess_BloomConvoution)) _astexture_2d_2))
            (Variables|/Script/Engine.PostProcessComponent|SetSettings _structout_2 (Variables|Default|GetPostProcess_BloomConvoution))))))
    (Rendering|SetVisibility _postprocess_bloomconvoution true false)
    (Development|SetHiddeninGame _postprocess_bloomconvoution false false)
    (Fortnite|FortShippingLog self "TODM Scalability: Convolution Bloom = True")
    (else
      (Rendering|SetVisibility _postprocess_bloomconvoution_2 false false)
      (Development|SetHiddeninGame _postprocess_bloomconvoution_2 true false)
      (bind _structout_3 (Utilities|Struct|SetmembersinPostProcessSettings (Variables|/Script/Engine.PostProcessComponent|GetSettings (Variables|Default|GetPostProcess_BloomConvoution)) "/Game/TimeOfDay/TODM_Sequence/Textures/Bloom/T_lowres_bloom_kernal.T_lowres_bloom_kernal"))
      (Variables|/Script/Engine.PostProcessComponent|SetSettings _structout_3 (Variables|Default|GetPostProcess_BloomConvoution))
      (Fortnite|FortShippingLog self "TODM Scalability: Convolution Bloom = False"))))

(event Custom|EventSetVolCloudMaterial
  (bind _volumetriccloudcomponent (Variables|DayNightCycle|GetVolumetricCloudComponent))
  (bind _volumetriccloudcomponent_2 (Variables|DayNightCycle|GetVolumetricCloudComponent))
  ;[0.2656 0.2656 0.2656 1 48] Set Volumetric Cloud Visibility and Material
  (Utilities|FlowControl|Sequence
    (:then_0
      (if (Class|BTechArtGlobalLibrary|UseVolumetricCloudsBR)
        ;;[0.3735 0.5512 1] Volumetric Cloud Material and Visibility
        (Rendering|SetVisibility _volumetriccloudcomponent true false)
        (Development|SetHiddeninGame _volumetriccloudcomponent false false)
        (Rendering|Material|SetScalarParameterValue "/Game/Packages/Fortress_SharedMaterials/GlobalMaterialParameters/FortniteMaterialParameters.FortniteMaterialParameters" "SkyBoxClouds" :!class /Script/Engine.KismetMaterialLibrary)
        (if (Deprecated|UEFN|IsinEditor)
          (bind _returnvalue (Utilities|LoadAssetBlocking (Variables|Clouds|GetVolumetricCloudMaterial)))
          (bind _asmaterial_interface (Utilities|Casting|CastToMaterialInterface _returnvalue))
          (Rendering|SetMaterial (Variables|DayNightCycle|GetVolumetricCloudComponent) _asmaterial_interface)
          (else
            (bind _object (Utilities|AsyncLoadAsset (Variables|Clouds|GetVolumetricCloudMaterial))
              (:Completed
                (bind _asmaterial_interface_2 (Utilities|Casting|CastToMaterialInterface _object))
                (Rendering|SetMaterial (Variables|DayNightCycle|GetVolumetricCloudComponent) _asmaterial_interface_2)))))
        (Fortnite|FortShippingLog self "TODM Scalability: Volumetric Clouds = True")
        (else
          (Rendering|SetMaterial (Variables|DayNightCycle|GetVolumetricCloudComponent) 0)
          (Rendering|SetVisibility _volumetriccloudcomponent_2 false false)
          (Development|SetHiddeninGame _volumetriccloudcomponent_2 true false)
          (Rendering|Material|SetScalarParameterValue "/Game/Packages/Fortress_SharedMaterials/GlobalMaterialParameters/FortniteMaterialParameters.FortniteMaterialParameters" "SkyBoxClouds" 1.0 :!class /Script/Engine.KismetMaterialLibrary)
          (Fortnite|FortShippingLog self "TODM Scalability: Volumetric Clouds =False"))))
    (:then_1)))

(event Custom|RotateTimeOfDay
  (params (Angle :type Float%20%28double-precision%29))
  ;[1 1 1] Cinematics Event For updating TODM Angle
  (Transformation|SetActorRotation (Math|Rotator|MakeRotator 0.0 0.0 Angle))
  (Fortnite|FortShippingLog self (Utilities|String|Append "DaySequenceActor Angle Updated: " (Utilities|String|ToString(Float) Angle))))

(event Custom|TODMSetExposureOverrides
  (CallFunction|SetExposureOverrides))

(fn ComputeCubeMapBlend
  (params (val :type Float%20%28double-precision%29)))

(fn IsExposureCvar
  (params (cvar :type String))
  (return (Math|Boolean|ORBoolean (Utilities|String|Equal,CaseInsensitive(String) cvar "ExposureOverride.Enable") (Utilities|String|Equal,CaseInsensitive(String) cvar "ExposureOverride.DayMin") (Utilities|String|Equal,CaseInsensitive(String) cvar "ExposureOverride.DayMax") (Utilities|String|Equal,CaseInsensitive(String) cvar "ExposureOverride.DayCompensation") (Utilities|String|Equal,CaseInsensitive(String) cvar "ExposureOverride.DaySkyLightLeaking") (Utilities|String|Equal,CaseInsensitive(String) cvar "ExposureOverride.NightMin") (Utilities|String|Equal,CaseInsensitive(String) cvar "ExposureOverride.NightMax") (Utilities|String|Equal,CaseInsensitive(String) cvar "ExposureOverride.NightCompensation") (Utilities|String|Equal,CaseInsensitive(String) cvar "ExposureOverride.NightSkyLightLeaking") (Utilities|String|Equal,CaseInsensitive(String) cvar "ExposureOverride.DayHighlightContrast") (Utilities|String|Equal,CaseInsensitive(String) cvar "ExposureOverride.DayShadowContrast") (Utilities|String|Equal,CaseInsensitive(String) cvar "ExposureOverride.NightHighlightContrast") (Utilities|String|Equal,CaseInsensitive(String) cvar "ExposureOverride.NightShadowContrast") (Utilities|String|Equal,CaseInsensitive(String) cvar "ExposureOverride.DayBlurredLuminanceBlend"))))

(fn isLumenSupported
  (bind _returnvalue (FortGlobals|IsRunninginCloud))
  (bind _returnvalue_2 (TouchPlatform|IsMobileGame))
  (bind _returnvalue_3 (not _returnvalue_2))
  (bind _returnvalue_4 (TouchPlatform|IsNintendoSwitch))
  (bind _returnvalue_5 (not _returnvalue_4))
  (bind _returnvalue_6 (or _returnvalue_3 _returnvalue))
  (bind _returnvalue_7 (Development|GetConsoleVariableIntValue "sg.GlobalIlluminationQuality"))
  (Fortnite|FortShippingLog self (Utilities|String|Append "TODM Scalability: Is Lumen Supported = " (Utilities|String|ToString(Boolean) (Math|Boolean|ANDBoolean (>= _returnvalue_7 2) _returnvalue_5 _returnvalue_6))))
  (return (Math|Boolean|ANDBoolean (>= _returnvalue_7 2) _returnvalue_5 _returnvalue_6)))

(fn SetCustomBloom
  (bind _postprocess_bloomconvoution (Variables|Default|GetPostProcess_BloomConvoution))
  (bind _returnvalue (Development|GetConsoleVariableIntValue "Athena.PostProcessLevel"))
  (if (>= _returnvalue 3)
    (Rendering|SetVisibility _postprocess_bloomconvoution true false)
    (Development|SetHiddeninGame _postprocess_bloomconvoution false false)
    (bind _returnvalue_2 (Utilities|LoadAssetBlocking (Variables|Bloom|GetConvolution%20Kernel)))
    (bind _astexture_2d (Utilities|Casting|CastToTexture2D _returnvalue_2))
    (bind _structout (Utilities|Struct|SetmembersinPostProcessSettings (Variables|/Script/Engine.PostProcessComponent|GetSettings (Variables|Default|GetPostProcess_BloomConvoution)) _astexture_2d))
    (Variables|/Script/Engine.PostProcessComponent|SetSettings _structout (Variables|Default|GetPostProcess_BloomConvoution))
    (Fortnite|FortShippingLog self "TODM Scalability: Convolution Bloom = True")
    (else
      (Rendering|SetVisibility _postprocess_bloomconvoution false false)
      (Development|SetHiddeninGame _postprocess_bloomconvoution true false)
      (bind _structout_2 (Utilities|Struct|SetmembersinPostProcessSettings (Variables|/Script/Engine.PostProcessComponent|GetSettings (Variables|Default|GetPostProcess_BloomConvoution)) ""))
      (Variables|/Script/Engine.PostProcessComponent|SetSettings _structout_2 (Variables|Default|GetPostProcess_BloomConvoution))
      (Fortnite|FortShippingLog self "TODM Scalability: Convolution Bloom = False"))))

(fn SetExposureOverrides
  (bind _postprocess_exposureoverridedaytime (Variables|Default|GetPostProcess_ExposureOverrideDaytime))
  (bind _postprocess_exposureoverridenight (Variables|Default|GetPostProcess_ExposureOverrideNight))
  (bind _postprocess_exposureoverridedaytime_2 (Variables|Default|GetPostProcess_ExposureOverrideDaytime))
  (bind _postprocess_exposureoverridedaytime_3 (Variables|Default|GetPostProcess_ExposureOverrideDaytime))
  (bind _postprocess_exposureoverridenight_2 (Variables|Default|GetPostProcess_ExposureOverrideNight))
  (bind _postprocess_exposureoverridenight_3 (Variables|Default|GetPostProcess_ExposureOverrideNight))
  (bind _postprocess_exposureoverridenight_4 (Variables|Default|GetPostProcess_ExposureOverrideNight))
  (bind _postprocess_exposureoverridedaytime_4 (Variables|Default|GetPostProcess_ExposureOverrideDaytime))
  (bind _islumensupported (CallFunction|IsLumenSupported))
  (if _islumensupported
    (bind _returnvalue (Development|GetConsoleVariableIntValue "ExposureOverride.Enable"))
    (if (> _returnvalue 0)
      (Rendering|SetVisibility _postprocess_exposureoverridedaytime_3 true false)
      (Development|SetHiddeninGame _postprocess_exposureoverridedaytime_3 false false)
      (Rendering|SetVisibility _postprocess_exposureoverridenight_2 true false)
      (Development|SetHiddeninGame _postprocess_exposureoverridenight_2 false false)
      (Utilities|FlowControl|Sequence
        (:then_0
          (bind _returnvalue_2 (Development|GetConsoleVariableFloatValue "ExposureOverride.DayMin"))
          (bind _returnvalue_3 (Development|GetConsoleVariableFloatValue "ExposureOverride.DayMax"))
          (bind _returnvalue_4 (Development|GetConsoleVariableFloatValue "ExposureOverride.DayCompensation"))
          (bind _returnvalue_5 (Development|GetConsoleVariableFloatValue "ExposureOverride.DaySkyLightLeaking"))
          (bind _returnvalue_6 (Development|GetConsoleVariableFloatValue "ExposureOverride.DayHighlightContrast"))
          (bind _returnvalue_7 (Development|GetConsoleVariableFloatValue "ExposureOverride.DayShadowContrast"))
          (bind _returnvalue_8 (Development|GetConsoleVariableFloatValue "ExposureOverride.DayBlurredLuminanceBlend"))
          (bind _structout (Utilities|Struct|SetmembersinPostProcessSettings (Variables|/Script/Engine.PostProcessComponent|GetSettings _postprocess_exposureoverridedaytime_4) _returnvalue_5 _returnvalue_4 _returnvalue_2 _returnvalue_3 _returnvalue_6 _returnvalue_7 _returnvalue_8))
          (Variables|/Script/Engine.PostProcessComponent|SetSettings _structout _postprocess_exposureoverridedaytime_4))
        (:then_1
          (bind _returnvalue_9 (Development|GetConsoleVariableFloatValue "ExposureOverride.NightMin"))
          (bind _returnvalue_10 (Development|GetConsoleVariableFloatValue "ExposureOverride.NightMax"))
          (bind _returnvalue_11 (Development|GetConsoleVariableFloatValue "ExposureOverride.NightCompensation"))
          (bind _returnvalue_12 (Development|GetConsoleVariableFloatValue "ExposureOverride.NightSkyLightLeaking"))
          (bind _returnvalue_13 (Development|GetConsoleVariableFloatValue "ExposureOverride.NightHighlightContrast"))
          (bind _returnvalue_14 (Development|GetConsoleVariableFloatValue "ExposureOverride.NightShadowContrast"))
          (bind _structout_2 (Utilities|Struct|SetmembersinPostProcessSettings (Variables|/Script/Engine.PostProcessComponent|GetSettings _postprocess_exposureoverridenight_4) _returnvalue_12 _returnvalue_11 _returnvalue_9 _returnvalue_10 _returnvalue_13 _returnvalue_14))
          (Variables|/Script/Engine.PostProcessComponent|SetSettings _structout_2 _postprocess_exposureoverridenight_4)))
      (else
        (Rendering|SetVisibility _postprocess_exposureoverridedaytime_2 false false)
        (Development|SetHiddeninGame _postprocess_exposureoverridedaytime_2 true false)
        (Rendering|SetVisibility _postprocess_exposureoverridenight false false)
        (Development|SetHiddeninGame _postprocess_exposureoverridenight true false)))
    (else
      (Rendering|SetVisibility _postprocess_exposureoverridedaytime false false)
      (Development|SetHiddeninGame _postprocess_exposureoverridedaytime true false)
      (Rendering|SetVisibility _postprocess_exposureoverridenight_3 false false)
      (Development|SetHiddeninGame _postprocess_exposureoverridenight_3 true false))))

(fn SetGlobalIlluminationMPC
  (bind _returnvalue (Settings|GetGameUserSettings))
  (Rendering|Material|SetScalarParameterValue "/Game/Packages/Fortress_SharedMaterials/GlobalMaterialParameters/FortniteMaterialParameters.FortniteMaterialParameters" "GlobalIlluminationEnabled" (Math|Conversions|ToFloat(Integer) (Math|Integer|Clamp(Integer) (- (Settings|GetGlobalIlluminationQuality _returnvalue) 1) 0 1)) :!class /Script/Engine.KismetMaterialLibrary)
  (Rendering|Material|SetScalarParameterValue "/Game/Packages/Fortress_SharedMaterials/GlobalMaterialParameters/FortniteMaterialParameters.FortniteMaterialParameters" "GlobalIlluminationEpicQuality" (Math|Conversions|ToFloat(Integer) (Math|Integer|Clamp(Integer) (- (Settings|GetGlobalIlluminationQuality _returnvalue) 2) 0 1)) :!class /Script/Engine.KismetMaterialLibrary)
  (Development|PrintString (Utilities|String|Append "GlobalIlluminationDebug " (Utilities|String|ToString(Integer) (Settings|GetGlobalIlluminationQuality _returnvalue)))))

(fn SetLowEndExposure
  (bind _postprocess_exposurelowend (Variables|Default|GetPostProcess_ExposureLowEnd))
  (bind _islumensupported (CallFunction|IsLumenSupported))
  (if _islumensupported
    (Rendering|SetVisibility _postprocess_exposurelowend false false)
    (Development|SetHiddeninGame _postprocess_exposurelowend true false)
    (Fortnite|FortShippingLog self "TODM Scalability: Low End Exposure =  False")
    (else
      (Rendering|SetVisibility _postprocess_exposurelowend true false)
      (Development|SetHiddeninGame _postprocess_exposurelowend false false)
      (Fortnite|FortShippingLog self "TODM Scalability: Low End Exposure = True"))))

(fn SetMotionBlurAmount
  (bind _returnvalue (Development|GetConsoleVariableIntValue "t.MaxFPS"))
  (Variables|/Script/Engine.PostProcessComponent|SetbEnabled (or (> _returnvalue 60) (== _returnvalue 0)) (Variables|Default|Get120fpsMotionBlurPostProcess)))

(fn SetReflectionMPC
  (bind _returnvalue (Settings|GetGameUserSettings))
  (Rendering|Material|SetScalarParameterValue "/Game/Packages/Fortress_SharedMaterials/GlobalMaterialParameters/FortniteMaterialParameters.FortniteMaterialParameters" "ReflectionsEnabled" (Math|Conversions|ToFloat(Integer) (Math|Integer|Clamp(Integer) (Settings|GetReflectionQuality _returnvalue) 0 1)) :!class /Script/Engine.KismetMaterialLibrary))

(fn SetShadowMPC
  (bind _returnvalue (Settings|GetGameUserSettings))
  (Rendering|Material|SetScalarParameterValue "/Game/Packages/Fortress_SharedMaterials/GlobalMaterialParameters/FortniteMaterialParameters.FortniteMaterialParameters" "ShadowsEnabled" (Math|Conversions|ToFloat(Integer) (Math|Integer|Clamp(Integer) (Settings|GetShadowQuality _returnvalue) 0 1)) :!class /Script/Engine.KismetMaterialLibrary))

(fn SetVolumetricCloudMaterial
  (bind _volumetriccloudcomponent (Variables|DayNightCycle|GetVolumetricCloudComponent))
  (bind _returnvalue (Development|GetConsoleVariableIntValue "Athena.VolumetricClouds.Quality"))
  (if (> _returnvalue 0)
    (bind _returnvalue_2 (Utilities|LoadAssetBlocking (Variables|Clouds|GetVolumetricCloudMaterial)))
    (bind _asmaterial_interface (Utilities|Casting|CastToMaterialInterface _returnvalue_2))
    (Rendering|SetMaterial (Variables|DayNightCycle|GetVolumetricCloudComponent) _asmaterial_interface)
    (Rendering|Material|SetScalarParameterValue "/Game/Packages/Fortress_SharedMaterials/GlobalMaterialParameters/FortniteMaterialParameters.FortniteMaterialParameters" "SkyBoxClouds" :!class /Script/Engine.KismetMaterialLibrary)
    (Fortnite|FortShippingLog self "TODM Scalability: Volumetric Clouds = True")
    (else
      (Rendering|SetMaterial (Variables|DayNightCycle|GetVolumetricCloudComponent) 0)
      (Rendering|SetVisibility _volumetriccloudcomponent false false)
      (Development|SetHiddeninGame _volumetriccloudcomponent true false)
      (Rendering|Material|SetScalarParameterValue "/Game/Packages/Fortress_SharedMaterials/GlobalMaterialParameters/FortniteMaterialParameters.FortniteMaterialParameters" "SkyBoxClouds" 1.0 :!class /Script/Engine.KismetMaterialLibrary)
      (Fortnite|FortShippingLog self "TODM Scalability: Volumetric Clouds = False"))))

(fn ToggleStormMeshVisibility
  (params (Visibility :type Boolean))
  ;[0 0.7534 1 0.2] Updates the mesh visibility for all actors that care (this is BP_SafeZoneStorm and BP_StormMeshRotator, if they exist)
  (bind _outactors (Actor|GetAllActorswithInterface "/Game/Athena/SafeZone/StormDaySequenceInterface.StormDaySequenceInterface_C"))
  (for _array_element _outactors
    (Storm/DaySequenceInterface|UpdateStormMeshVisibility(Message) _array_element Visibility)))

(fn UpdateComponentVisibility
  ;[1 1 1] Functions that toggle component visibility
  (CallFunction|SetLowEndExposure)
  (CallFunction|SetExposureOverrides)
  (CallFunction|EventSetVolCloudMaterial)
  (CallFunction|EventSetCustomBloom)
  (CallFunction|SetShadowMPC)
  (CallFunction|SetReflectionMPC)
  (CallFunction|SetGlobalIlluminationMPC))

(fn UpdateScalability
  (Default|SetMotionBlurAmount)
  (Fortnite|FortShippingLog self (Utilities|String|Append "TODM Scalability: TODM Visibility = " (Utilities|String|ToString(Boolean) (TimeOfDay|IsVisible)))))

(fn ConstructionScript
  (if (Editor|InEditor)
    (CallFunction|EventSetVolCloudMaterial)
    (CallFunction|EventSetCustomBloom)))
