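; Lighting actor, apparently for the spawn island, that holds sky, fog, sun and post-process components with low-end alternates.
; On BeginPlay and whenever game user settings signal a UI update, it flushes net dormancy and re-evaluates scalability.
; SetLowEndExposure toggles skylight, directional light and low-end post-process visibility by mobile or Switch platform and global illumination quality, logging the result.
; A "Disable all" event hides the components and disables both post-processes.
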(event EventBeginPlay
  (CallFunction|UpdateScalability)
  (bind _returnvalue (Settings|GetGameUserSettings))
  (EventDispatchers|BindEventtoOnGameUserSettingsUINeedsUpdate _returnvalue (EventDispatchers|CreateEvent :!event Custom|UpdateLowEndScalability)))

(event Collision|EventActorBeginOverlap
  (params (OtherActor :type Actor%20Object%20Reference)))

(event Custom|UpdateLowEndScalability
  (CallFunction|UpdateScalability))

(event EventTick
  (params (DeltaSeconds :type Float%20%28single-precision%29)))

(event Custom|Disable%20all%20
  (bind _postprocess_01 (Variables|Default|GetPostProcess_01))
  (bind _postprocess_lowend (Variables|Default|GetPostProcess_LowEnd))
  (Rendering|SetVisibility _postprocess_01 false false)
  (Variables|/Script/Engine.PostProcessComponent|SetbEnabled false _postprocess_01)
  (Variables|/Script/Engine.PostProcessComponent|SetbEnabled false _postprocess_lowend)
  (Rendering|SetVisibility _postprocess_lowend false false)
  (Rendering|SetVisibility (Variables|Default|GetSkyAtmosphere) false false)
  (Rendering|SetVisibility (Variables|Default|GetExponentialHeightFog) false false)
  (Rendering|SetVisibility (Variables|Default|GetDirectionalLight) false false)
  (Rendering|SetVisibility (Variables|Default|GetDefaultSceneRoot) false true)
  (Rendering|SetVisibility (Variables|Default|GetSkyLight) false false)
  (Rendering|SetVisibility (Variables|Default|GetSkyLight_lowend) false false)
  (Rendering|SetVisibility (Variables|Default|GetDirectionalLight_lowend) false false))

(fn SetLowEndExposure
  (if (not (or (TouchPlatform|IsMobileGame) (TouchPlatform|IsNintendoSwitch)))
    (bind _returnvalue (Development|GetConsoleVariableIntValue "sg.GlobalIlluminationQuality"))
    (if (>= _returnvalue 2)
      (Rendering|SetVisibility (Variables|Default|GetPostProcess_LowEnd) false false)
      (Rendering|SetVisibility (Variables|Default|GetSkyLight) true false)
      (Rendering|SetVisibility (Variables|Default|GetSkyLight_LowEnd) false false)
      (Rendering|SetVisibility (Variables|Default|GetDirectionalLight_lowend) false false)
      (Rendering|SetVisibility (Variables|Default|GetDirectionalLight) true false)
      (Fortnite|FortShippingLog self "TODM Scalability: Low End Exposure =  False")
      (else
        _)))
  (Rendering|SetVisibility (Variables|Default|GetPostProcess_LowEnd) true false)
  (Rendering|SetVisibility (Variables|Default|GetSkyLight) false false)
  (Rendering|SetVisibility (Variables|Default|GetSkyLight_LowEnd) true false)
  (Rendering|SetVisibility (Variables|Default|GetDirectionalLight_lowend) true false)
  (Rendering|SetVisibility (Variables|Default|GetDirectionalLight) false false)
  (Fortnite|FortShippingLog self "TODM Scalability: Low End Exposure = True"))

(fn UpdateScalability
  (Networking|FlushNetDormancy)
  (CallFunction|SetLowEndExposure))

(fn ConstructionScript)
