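; BP_TODM is a time-of-day Blueprint that offsets the sky atmosphere's light direction from the sun component's actual rotation.
; OverrideSkyAtmosphereLightDir stores the offset, adds it to the sun's world pitch, then passes the negated, normalized forward vector to the SkyAtmosphere component's override for light 0.
; The construction script applies the saved offset, and the SetSkyAtmosphereLightOffset event applies new values.
; BeginPlay and Tick are empty, so the override does not follow sun rotation per frame here.
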(event EventBeginPlay)

(event EventTick
  (params (DeltaSeconds :type Float%20%28single-precision%29)))

(event Custom|SetSkyAtmosphereLightOffset
  (params (offset :type Float%20%28double-precision%29))
  (CallFunction|OverrideSkyAtmosphereLightDir offset))

(fn OverrideSkyAtmosphereLightDir
  (params (offset :type Float%20%28double-precision%29))
  (bind _output_get (Variables|Default|SetSkyAtmosphereLightOffset offset))
  (Rendering|OverrideAtmosphereLightDirection (Variables|DayNightCycle|GetSkyAtmosphereComponent) 0 (Math|Vector|Normalize (* (Math|Vector|GetForwardVector (Math|Rotator|MakeRotator (.roll (Transformation|GetWorldRotation (Variables|DayNightCycle|GetSunComponent))) (+ _output_get (.pitch (Transformation|GetWorldRotation (Variables|DayNightCycle|GetSunComponent)))) (.yaw (Transformation|GetWorldRotation (Variables|DayNightCycle|GetSunComponent))))) "-1.000000,-1.000000,-1.000000"))))

(fn ConstructionScript
  (CallFunction|OverrideSkyAtmosphereLightDir (Variables|Default|GetSkyAtmosphereLightOffset)))
