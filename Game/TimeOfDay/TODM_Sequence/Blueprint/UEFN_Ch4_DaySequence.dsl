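; UEFN Chapter 4 day sequence Blueprint for time-of-day lighting that feeds sun and fog state to materials.
; SetMPCValues writes real time, sun color, fog inscattering colors and sun axes, with forward negated, into the FortniteMaterialParameters collection.
; Nothing in this graph calls SetMPCValues.
; The scalability sequence event calls UpdateScalability, which only defers to the parent; overlap and tick events do nothing.
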(event Collision|EventActorBeginOverlap
  (params (OtherActor :type Actor%20Object%20Reference)))

(event EventTick
  (params (DeltaSeconds :type Float%20%28single-precision%29))
  (|CallParentFunction DeltaSeconds) ; This node is disabled and will not be called.\nDrag off pins to build functionality.
)

(event Sequence|EventUpdateScalabilitySequences
  (CallFunction|UpdateScalability))

(fn SetMPCValues
  (Rendering|Material|SetScalarParameterValue "/Game/Packages/Fortress_SharedMaterials/GlobalMaterialParameters/FortniteMaterialParameters.FortniteMaterialParameters" "RealGameTime" (Utilities|Time|GetRealTimeSeconds) :!class /Script/Engine.KismetMaterialLibrary)
  (Rendering|Material|SetVectorParameterValue "/Game/Packages/Fortress_SharedMaterials/GlobalMaterialParameters/FortniteMaterialParameters.FortniteMaterialParameters" "SunLightColor" (Math|Conversions|ToLinearColor(Color) (Variables|/Script/Engine.LightComponentBase|GetLightColor (Variables|DayNightCycle|GetSunComponent))) :!class /Script/Engine.KismetMaterialLibrary)
  (Rendering|Material|SetVectorParameterValue "/Game/Packages/Fortress_SharedMaterials/GlobalMaterialParameters/FortniteMaterialParameters.FortniteMaterialParameters" "FogColor" (Variables|/Script/Engine.ExponentialHeightFogComponent|GetFogInscatteringLuminance (Variables|DayNightCycle|GetExponentialHeightFogComponent)) :!class /Script/Engine.KismetMaterialLibrary)
  (Rendering|Material|SetVectorParameterValue "/Game/Packages/Fortress_SharedMaterials/GlobalMaterialParameters/FortniteMaterialParameters.FortniteMaterialParameters" "FogDirectionalInscatteringColor" (Variables|/Script/Engine.ExponentialHeightFogComponent|GetDirectionalInscatteringLuminance (Variables|DayNightCycle|GetExponentialHeightFogComponent)) :!class /Script/Engine.KismetMaterialLibrary)
  (Rendering|Material|SetVectorParameterValue "/Game/Packages/Fortress_SharedMaterials/GlobalMaterialParameters/FortniteMaterialParameters.FortniteMaterialParameters" "SunlightYTransformVector" (Math|Conversions|ToLinearColor(Vector) (Transformation|GetRightVector (Variables|DayNightCycle|GetSunComponent))) :!class /Script/Engine.KismetMaterialLibrary)
  (Rendering|Material|SetVectorParameterValue "/Game/Packages/Fortress_SharedMaterials/GlobalMaterialParameters/FortniteMaterialParameters.FortniteMaterialParameters" "SunlightZTransformVector" (Math|Conversions|ToLinearColor(Vector) (Transformation|GetUpVector (Variables|DayNightCycle|GetSunComponent))) :!class /Script/Engine.KismetMaterialLibrary)
  (Rendering|Material|SetVectorParameterValue "/Game/Packages/Fortress_SharedMaterials/GlobalMaterialParameters/FortniteMaterialParameters.FortniteMaterialParameters" "SunlightDirectionalVector" (Math|Conversions|ToLinearColor(Vector) (* (Transformation|GetForwardVector (Variables|DayNightCycle|GetSunComponent)) -1.0)) :!class /Script/Engine.KismetMaterialLibrary))

(fn UpdateScalability
  (|CallParentFunction))

(fn ConstructionScript
  (|CallParentFunction))
